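; Base actor for HLOD proxy stand-ins in Athena's HLOD setup, with all logic in the construction script.
; When the color override flag is set, it creates a dynamic material instance on slot 0 of its static mesh and writes the override into the Color parameter.
; It then turns off collision and shadow casting on every primitive component, while BeginPlay, overlap and Tick events are left empty.
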(event EventBeginPlay)

(event Collision|EventActorBeginOverlap
  (params (OtherActor :type Actor%20Object%20Reference)))

(event EventTick
  (params (DeltaSeconds :type Float%20%28single-precision%29)))

(fn ConstructionScript
  (Utilities|FlowControl|Sequence
    (:then_0
      (if (Variables|Default|GetoverrideColor)
        (bind _returnvalue (Rendering|Material|CreateDynamicMaterialInstance (Variables|Default|GetStaticMesh) 0 0 "HLODPoxyColor" :!class /Script/Engine.PrimitiveComponent))
        (Rendering|Material|SetVectorParameterValue _returnvalue "Color" (Variables|Default|GetColorOverride) :!class /Script/Engine.MaterialInstanceDynamic)))
    (:then_1
      (for _array_element (Actor|GetComponentsByClass "/Script/Engine.PrimitiveComponent")
        (Collision|SetCollisionEnabled _array_element "NoCollision")
        (Rendering|SetCastShadow _array_element false)))))
