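# M_HLODProxyActor is a Default Lit material in the Athena HLOD setup folder, named for HLOD proxy actors.
# It holds a Color vector parameter defaulting to a very dark grey, plus two scalar constants of zero and one.
# It samples no textures, and the graph shows no links from these nodes to material inputs.
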Material /Game/Athena/HLODSetup/Actors/M_HLODProxyActor
graph {
 properties {pixelDepthOffsetMode=PDOM_Legacy, usedShadingModels=Default Lit}
 nodes [
  Constant()
  Constant() {r=1}
  VectorParameter() {parameterName=Color, defaultValue=<rgba 0.0156 0.0156 0.0156 1>}
 ]
}
